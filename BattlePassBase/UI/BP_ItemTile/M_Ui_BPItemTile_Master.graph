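# Translucent UI master material for Battle Pass item tiles, drawing the offer image over a background colored by free, premium and owned state.
# Image UVs get zoom, offset, rotation and hover motion, feeding a texture or flipbook sample with a color falloff.
# The background combines a radial gradient, panning screen-aligned triangles and a Progress-driven triangle wipe, with bIsS23 switching blend styles.
# IsPremiumTile drops the background and masks opacity to the image.
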Material /BattlePassBase/UI/BP_ItemTile/M_Ui_BPItemTile_Master
graph {
 properties {blendMode=BLEND_Translucent, materialDomain=MD_UI, pixelDepthOffsetMode=PDOM_Legacy, usedShadingModels=Default Lit}
 nodes [
  n1 = Constant()
  n2 = ScalarParameter() {defaultValue=25, group=Image, parameterName=FallOffColor_Postion}
  n3 = Multiply(A=LinearInterpolate(Alpha=Multiply(A=@n2) {constB=0.01}) {constA=1, constB=0}) {constB=-1}
  n4 = MaterialFunctionCall() {materialFunction=&GetUserInterfaceUV}
  n5 = MaterialFunctionCall(Float2=@n4.Pixel Size) {materialFunction=&BreakOutFloat2Components}
  n6 = Divide(A=@n5.R, B=@n5.R)
  n7 = Constant() {r=1}
  n8 = OneMinus(None=@n6)
  n9 = Add(A=Multiply(A=@n4.Normalized UV, B=AppendVector(A=@n6, B=@n7)), B=AppendVector(A=Multiply(A=@n8) {constB=0.5}, B=Constant()))
  n10 = ComponentMask(None=Add(A=@n9, B=AppendVector(A=@n1, B=@n3))) {g=true}
  n11 = ScalarParameter() {defaultValue=1.22, group=Image, parameterName=FallOffColor_DecayRate}
  n12 = ScalarParameter() {defaultValue=40, group=Image, parameterName=FallOffColor_Fill_Percent}
  n13 = ScalarParameter() {group=Image, parameterName=FallOffColor_Fill_Percent_Hovered}
  n14 = ScalarParameter() {parameterName=Hover, sliderMax=1}
  n15 = Multiply(A=Saturate(None=Power(Base=@n10, Exp=@n11)), B=Multiply(A=LinearInterpolate(A=@n12, B=@n13, Alpha=@n14)) {constB=0.01})
  n16 = Power(Base=ScalarParameter() {group=Progress, parameterName=Progress, sliderMax=1}) {constExponent=3}
  n17 = ScalarParameter() {defaultValue=7.5, group=Animation, parameterName=Hover_Zoom_Increment}
  n18 = MaterialFunctionCall(Float2=MaterialFunctionCall() {materialFunction=&GetUserInterfaceUV}.Pixel Size) {materialFunction=&BreakOutFloat2Components}
  n19 = Constant()
  n20 = If(A=@n18.R, B=@n18.R, "A > B"=@n19, "A == B"=@n19, "A < B"=Constant() {r=1})
  n21 = Add(
   A=ScalarParameter() {defaultValue=50, group=Image, parameterName=ZoomImage_Percent, sliderMax=800, sliderMin=-100}
   B=LinearInterpolate(B=ScalarParameter() {group=Image, parameterName=Scale_Compensation}, Alpha=@n20) {constB=7.5}
  )
  n22 = LinearInterpolate(A=@n21, B=Add(A=@n21, B=@n17) {constB=10}, Alpha=ScalarParameter() {parameterName=Hover, sliderMax=1}) {constA=0.95}
  n23 = LinearInterpolate(B=Subtract(B=@n17) {constA=100}, Alpha=LinearInterpolate(Alpha=Multiply(A=@n22) {constB=0.01}) {constB=0.325}) {constAlpha=100, constB=95}
  n24 = Add(A=@n23, B=ScalarParameter() {group=Animation, parameterName=Hover_Zoom_Increment_2})
  n25 = ScalarParameter() {parameterName=Hover, sliderMax=1}
  n26 = Add(
   A=Multiply(A=LinearInterpolate(A=@n23, B=@n24, Alpha=@n25) {constA=0.95}) {constB=0.01}
   B=LinearInterpolate(
    B=ScalarParameter() {defaultValue=0.5, group=Progress, parameterName=ProgressZoomValue}
    Alpha=ScalarParameter() {group=Progress, parameterName=Progress_Zoom, sliderMax=1}
   ) {constB=0.25}
  )
  n27 = Clamp(None=LinearInterpolate(
    A=ScalarParameter() {defaultValue=1, group=ZoomCrop, parameterName=Zoom_Min}
    B=ScalarParameter() {defaultValue=2, group=ZoomCrop, parameterName=Zoom_Max}
    Alpha=@n26
   )) {maxDefault=2}
  n28 = AppendVector(
   A=ScalarParameter() {defaultValue=0.5, group=ZoomCrop, parameterName=ZoomAroundPoint_X}
   B=LinearInterpolate(A=Constant(), B=ScalarParameter() {defaultValue=0.125, group=ZoomCrop, parameterName=ZoomAroundPoint_Y}, Alpha=@n26)
  )
  n29 = Add(A=Divide(A=@n9, B=@n27), B=@n28)
  n30 = Divide(A=@n28, B=@n27) {desc=Scale the center offset by the scaling factor}
  n31 = ScalarParameter() {group=Image, parameterName=OffsetImage_X_Compensation}
  n32 = Multiply(A=Add(A=ScalarParameter() {group=Image, parameterName=OffsetImage_X}, B=LinearInterpolate(B=@n31, Alpha=@n20))) {constB=-0.01}
  n33 = ScalarParameter() {group=Image, parameterName=OffsetImage_Y, sliderMax=100, sliderMin=-100}
  n34 = Add(A=@n33, B=ScalarParameter() {defaultValue=3, parameterName=free_Ycompensation})
  n35 = ScalarParameter() {parameterName=IsFree, sliderMax=1}
  n36 = ScalarParameter() {group=Image, parameterName=OffsetImage_Y_Compensation}
  n37 = Multiply(A=Add(A=LinearInterpolate(A=@n33, B=@n34, Alpha=@n35), B=LinearInterpolate(B=@n36, Alpha=@n20))) {constB=-0.01}
  n38 = MaterialFunctionCall(Float2=MaterialFunctionCall() {materialFunction=&GetUserInterfaceUV}.Pixel Size) {materialFunction=&BreakOutFloat2Components}
  n39 = Floor(None=Divide(A=@n38.R, B=Max(A=@n38.G)))
  n40 = Rotator(
   Coordinate=Add(A=Subtract(A=@n29, B=@n30) {desc=Re offset to maintain center}, B=AppendVector(A=@n32, B=@n37))
   Time=LinearInterpolate(
    A=ScalarParameter() {group=Image, parameterName=Angle}
    B=ScalarParameter() {group=Image, parameterName=Angle_Compensation}
    Alpha=Saturate(None=@n39)
   )
  ) {speed=6.2832}
  n41 = Multiply(A=ScalarParameter() {defaultValue=0.2, parameterName=Hovering_Value}) {constB=0.001}
  n42 = Multiply(A=@n41) {constB=-1}
  n43 = Sine(None=Multiply(A=Time(), B=ScalarParameter() {defaultValue=0.25, parameterName=hoveringtime}))
  n44 = MaterialFunctionCall(
   X=LinearInterpolate(B=LinearInterpolate(A=@n42, B=@n41, Alpha=@n43), Alpha=ScalarParameter() {parameterName=ISHovering, sliderMax=1})
   Y=Constant()
  ) {materialFunction=&MakeFloat2}
  n45 = AppendVector(A=ScalarParameter() {group=Image, parameterName=Premium_OffsetImageX}, B=ScalarParameter() {group=Image, parameterName=Premium_OffsetImageY})
  n46 = ScalarParameter() {group=Image, parameterName=IsPremiumTile}
  n47 = MaterialFunctionCall(
   UVs=Add(A=Add(A=@n40, B=@n44.Result), B=LinearInterpolate(B=@n45, Alpha=@n46))
   Texture Scale=LinearInterpolate(
    B=ScalarParameter() {defaultValue=1, group=Image, parameterName=Premium_ZoomOverride}
    Alpha=ScalarParameter() {group=Image, parameterName=IsPremiumTile}
   ) {constA=1}
  ) {materialFunction=&ScaleUVsByCenter}
  n48 = TextureObjectParameter() {group=Image, parameterName=OfferImage, texture=/Game/Athena/Prototype/Textures/T_Placeholder_Item_Outfit-S}
  n49 = TextureSample(UVs=@n47.UVs, Tex=@n48) {samplerSource=SSM_Clamp_WorldGroupSettings}
  n50 = MaterialFunctionCall(
   "Animation  Phase (0-1)"=Multiply(A=Time(), B=ScalarParameter() {defaultValue=1, group=FlipBook, parameterName=Speed})
   Number of Rows=ScalarParameter() {defaultValue=0.5, group=FlipBook, parameterName=Rows}
   Number of Columns=ScalarParameter() {defaultValue=4, group=FlipBook, parameterName=Columns}
   Texture=@n48
   UVs=@n47.UVs
  ) {materialFunction=&FlipBook}
  n51 = ScalarParameter() {group=FlipBook, parameterName=Use FLipbook}
  n52 = LinearInterpolate(
   A=Add(A=LinearInterpolate(A=@n49.RGB, B=@n50.Result, Alpha=@n51), B=LinearInterpolate(Alpha=@n16))
   B=LinearInterpolate(
    A=VectorParameter() {group=Image, parameterName=FallOff_Color, defaultValue=<rgba 0.0168 0.2122 0.6654 1>}.RGB
    B=VectorParameter() {group=Image, parameterName=FallOff_Color_Free, defaultValue=<rgba 0.2469 0.254 0.2708 1>}.RGB
    Alpha=ScalarParameter() {parameterName=IsFree, sliderMax=1}
   )
   Alpha=@n15
  )
  n53 = LinearInterpolate(A=@n49.A, B=@n50.Alpha, Alpha=ScalarParameter() {group=FlipBook, parameterName=Use FLipbook})
  n54 = MaterialFunctionCall!(UVs=@n47.UVs, Alpha=@n53) {materialFunction=/Game/UI/Foundation/Items/ClampAlphaToTransparentBorder}
  n55 = ScalarParameter() {defaultValue=50, group=Background Gradient, parameterName=Gradient_Position_X}
  n56 = ScalarParameter() {defaultValue=25, group=Background Gradient, parameterName=Gradient_Position_Y}
  n57 = ScalarParameter() {defaultValue=16, group=Background Gradient, parameterName=Gradient_Hardness}
  n58 = Add(A=@n57, B=Multiply(A=@n57) {constB=0.25})
  n59 = ScalarParameter() {group=Animation, parameterName=HoverBurst}
  n60 = MaterialFunctionCall() {materialFunction=&GetUserInterfaceUV}
  n61 = MaterialFunctionCall(Float2=@n60.Pixel Size) {materialFunction=&BreakOutFloat2Components}
  n62 = Divide(A=@n61.G, B=@n61.G)
  n63 = Multiply(A=@n60.Normalized UV, B=AppendVector(A=Constant() {r=1}, B=@n62))
  n64 = AppendVector(A=Constant(), B=Multiply(A=OneMinus(None=@n62)) {constB=0.5})
  n65 = MaterialFunctionCall(
   UVs=LinearInterpolate(A=Add(A=@n63, B=@n64), B=@n9, Alpha=@n20)
   CenterPosition=AppendVector(A=Multiply(A=@n55) {constB=0.01}, B=Multiply(A=@n56) {constB=0.01})
   Radius=Multiply(A=ScalarParameter() {defaultValue=66.67, group=Background Gradient, parameterName=Gradient_Size}) {constB=0.01}
   Density=Multiply(A=LinearInterpolate(A=@n57, B=@n58, Alpha=@n59)) {constB=0.1}
  ) {materialFunction=&RadialGradientExponential}
  n66 = LinearInterpolate(
   A=LinearInterpolate(
    A=VectorParameter() {group=Background Gradient, parameterName=Premium_B, defaultValue=<rgba 0.0809 0.5107 0.8177 1>}.RGB
    B=VectorParameter() {group=Background Gradient, parameterName=Premium_Owned_B, defaultValue=<rgba 0.25 0.6875 1 1>}.RGB
    Alpha=ScalarParameter() {parameterName=IsOwned, sliderMax=1}
   )
   B=LinearInterpolate(
    A=VectorParameter() {group=Background Gradient, parameterName=Free_B, defaultValue=<rgba 0.4814 0.5105 0.5312 1>}.RGB
    B=VectorParameter() {group=Background Gradient, parameterName=Free_Owned_B, defaultValue=<rgba 0.4909 0.5205 0.5417 1>}.RGB
    Alpha=ScalarParameter() {parameterName=IsOwned, sliderMax=1}
   )
   Alpha=ScalarParameter() {parameterName=IsFree, sliderMax=1}
  )
  n67 = LinearInterpolate(
   A=LinearInterpolate(
    A=VectorParameter() {group=Background Gradient, parameterName=Premium_A, defaultValue=<rgba 0 0.0794 0.4688 1>}.RGB
    B=VectorParameter() {group=Background Gradient, parameterName=Premium_Owned_A, defaultValue=<rgba 0.0167 0.2115 0.666 1>}.RGB
    Alpha=ScalarParameter() {parameterName=IsOwned, sliderMax=1}
   )
   B=LinearInterpolate(
    A=VectorParameter() {group=Background Gradient, parameterName=Free_A, defaultValue=<rgba 0.1282 0.1319 0.1406 1>}.RGB
    B=VectorParameter() {group=Background Gradient, parameterName=Free_Owned_A, defaultValue=<rgba 0.2469 0.254 0.2708 1>}.RGB
    Alpha=ScalarParameter() {parameterName=IsOwned, sliderMax=1}
   )
   Alpha=ScalarParameter() {parameterName=IsFree, sliderMax=1}
  )
  n68 = LinearInterpolate(A=@n67, B=@n66, Alpha=@n65.RadialGradientExponential)
  n69 = MaterialFunctionCall(Float2=MaterialFunctionCall() {materialFunction=&ScreenResolution}.Visible Resolution) {materialFunction=&BreakOutFloat2Components}
  n70 = Max(A=@n69.G)
  n71 = Multiply(
   A=AppendVector(A=Divide(A=@n69.R, B=@n70), B=Constant() {r=1})
   B=MaterialFunctionCall() {materialFunction=&ScreenAlignedUVs}."X 100%, Y 100%"
  )
  n72 = ScalarParameter() {defaultValue=-0.0571, parameterName=Tri Angle}
  n73 = ScalarParameter() {defaultValue=0.0025, parameterName=PanSpeed}
  n74 = Constant()
  n75 = Multiply(
   A=Panner(Coordinate=Rotator(Coordinate=@n71, Time=@n72) {speed=6.2832}, Speed=AppendVector(A=@n73, B=@n74))
   B=ScalarParameter() {defaultValue=4, parameterName=Tri Tile Amount}
  )
  n76 = LinearInterpolate(
   A=@n67
   B=@n66
   Alpha=Multiply(
    A=TextureSample(UVs=@n75) {texture=/Game/Global/Textures/Default/Masks_and_Gradients/T-TriangleGrid}.R
    B=ScalarParameter() {defaultValue=0.2, parameterName=TriangleDensity, sliderMax=1}
   )
  )
  n77 = Saturate(None=OneMinus(None=@n65.RadialGradientExponential))
  n78 = Multiply(A=@n76, B=ScalarParameter() {defaultValue=0.5})
  n79 = ScalarParameter() {defaultValue=-0.01, group=Progress, parameterName=ProgressValue_start, sliderMax=1, sliderMin=-1}
  n80 = ScalarParameter() {defaultValue=1, group=Progress, parameterName=ProgressValue_End, sliderMax=1, sliderMin=-1}
  n81 = ScalarParameter() {group=Progress, parameterName=Progress, sliderMax=1}
  n82 = ComponentMask(None=Add(
    A=MaterialFunctionCall() {materialFunction=&GetUserInterfaceUV}.Normalized UV
    B=LinearInterpolate(A=@n79, B=@n80, Alpha=@n81) {constA=-1, constB=0}
   )) {r=true}
  n83 = ScalarParameter() {defaultValue=512, group=Progress, parameterName=Progress strenght}
  n84 = Saturate(None=OneMinus(None=Multiply(A=@n82, B=@n83)))
  n85 = Saturate(None=MaterialFunctionCall!(Blend=Multiply(A=@n76, B=@n77), Base=@n68) {materialFunction=/Engine/Functions/Engine_MaterialFunctions03/Blends/Blend_ColorDodge}.Result)
  n86 = MaterialFunctionCall(
   Blend=LinearInterpolate(
    A=VectorParameter() {group=Background Gradient, parameterName=Premium_Owned_B, defaultValue=<rgba 0.25 0.6875 1 1>}.RGB
    B=VectorParameter() {group=Background Gradient, parameterName=Free_Owned_B, defaultValue=<rgba 0.4909 0.5205 0.5417 1>}.RGB
    Alpha=ScalarParameter() {parameterName=IsFree, sliderMax=1}
   )
   Base=@n85
  ) {materialFunction=&Blend_ColorDodge}
  n87 = StaticSwitchParameter(
   True=LinearInterpolate(A=LinearInterpolate(A=@n68, B=@n76, Alpha=@n77), B=Add(A=@n78, B=@n68), Alpha=@n84)
   False=LinearInterpolate(A=@n85, B=Saturate(None=@n86.Result), Alpha=@n84)
  ) {defaultValue=true, parameterName=bIsS23}
  n88 = MaterialFunctionCall(Float2=MaterialFunctionCall() {materialFunction=&ScreenResolution}.Visible Resolution) {materialFunction=&BreakOutFloat2Components}
  n89 = Divide(A=@n88.R, B=Max(A=@n88.G))
  n90 = Constant() {r=1}
  n91 = Multiply(
   A=Multiply(
    A=AppendVector(A=@n89, B=@n90)
    B=MaterialFunctionCall() {materialFunction=&ScreenAlignedUVs}."X 100%, Y 100%"
   )
   B=ScalarParameter() {defaultValue=10, group=FX-Triangle, parameterName=FX-triangleTile}
  ) {constB=3}
  n92 = AppendVector(A=Multiply(A=Time()) {constB=0.25}, B=Constant())
  n93 = TextureSample(UVs=Rotator(Coordinate=Add(A=@n91, B=@n92), Time=ScalarParameter() {defaultValue=0.125, group=FX-Triangle, parameterName=FX-Triangle-Rotation}) {speed=6.2832}) {texture=/Game/Global/Textures/Default/Masks_and_Gradients/T-BattlePass-Tile-Triangles}
  n94 = Subtract(A=@n82) {constB=-1}
  n95 = OneMinus(None=@n94)
  n96 = Saturate(None=Multiply(A=Subtract(A=@n93.RGB, B=@n95) {constB=0}) {constB=32})
  n97 = Power(Base=@n94, Exp=ScalarParameter() {defaultValue=2, group=FX-Triangle, parameterName=FX-Triangle-Falloff-Power})
  n98 = ScalarParameter() {defaultValue=16, group=FX-Triangle, parameterName=FX-Triangle-Edge-Falloff-Power}
  n99 = Saturate(None=Add(A=Multiply(A=@n96, B=@n97), B=Power(Base=@n94, Exp=@n98)))
  n100 = LinearInterpolate(B=@n87, Alpha=Saturate(None=LinearInterpolate(B=@n99, Alpha=@n84)))
  n101 = ScalarParameter() {defaultValue=1, group=FX-Triangle, parameterName=FX-Triangle-Final-Intensity}
  n102 = LinearInterpolate(A=Add(A=@n87, B=Multiply(A=@n100, B=@n101)), Alpha=ScalarParameter() {parameterName=ToWhite})
  n103 = ScalarParameter() {group=Image, parameterName=IsPremiumTile}
  n104 = VectorParameter() {parameterName=OverlayTintColor}
  LinearInterpolate(A=LinearInterpolate(A=LinearInterpolate(A=@n102, B=@n52, Alpha=@n54.Result), B=@n52, Alpha=@n103), B=@n104.RGB, Alpha=@n104.RGB)
  n105 = LinearInterpolate(B=Add(A=@n15, B=@n53), Alpha=@n54.Result)
  n106 = MaterialFunctionCall(
   UVs=TextureCoordinate() {coordinateIndex=4}
   Texture Scale=Constant() {desc="Image in Tile is scaled up 2x (1/Transform.x) is what this should equal", r=0.5}
  ) {materialFunction=&ScaleUVsByCenter}
  n107 = ComponentMask(None=@n106.UVs) {r=true}
  n108 = OneMinus(None=@n107)
  n109 = OneMinus(None=ComponentMask(None=@n106.UVs) {g=true})
  n110 = Constant() {desc=Rare screen aspect ratios this mask is off by 1 pixel if set to 0, r=0.01}
  n111 = Multiply(A=Ceil(None=Multiply(A=@n108, B=@n107)), B=Ceil(None=Subtract(A=@n109, B=@n110) {constB=0.25}))
  LinearInterpolate(A=Constant() {r=1}, B=Multiply(A=Saturate(None=@n105), B=Saturate(None=@n111)), Alpha=@n103)
 ]
}
